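# Physics asset for the BalloonFloat gadget prop mesh, a cosmetic animation prop.
# It holds a single body on the root bone, shaped as one large, short capsule that is slightly offset and rotated from the bone.
# There are no constraints, so the asset provides only a simple rough collision volume with no jointed ragdoll parts.
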PhysicsAsset /BRCosmetics/Animation/Props/Gadgets_BalloonFloat/Meshes/PA_BalloonFloat_Gadget
bodies {root [{shapeType=Capsule, length=50.5974, radius=297.6317, center=<-10.3883 41.9717 0.4678>, rotation=<rot 0 107.4017 -89.2574>}]}
constraints []
